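# Material function for the Master Key backpack FX that hue-shifts emissive color.
# It pulls EmissiveColor from the input material attributes, runs it through the HueShift function, and writes it back to the MA output.
# A separate InputEmissive input gets the same shift and leaves as the Emissive output.
# The shift amount defaults to the HueShift channel of a Dynamic Parameter, allowing particle-driven control when that input is unconnected.
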MaterialFunction /Game/Accessories/FORT_Backpacks/Backpack_Master_Key/FX/MF_ColorShift
graph {nodes [
  n1 = GetMaterialAttributes(None=FunctionInput() {inputName=MA, inputType=FunctionInput_MaterialAttributes, sortPriority=1}) {attributeGetTypes [#1]}
  n2 = FunctionInput(Preview=DynamicParameter() {paramNames [Param1,Param2,Param3,HueShift]}.HueShift) {bUsePreviewValueAsDefault=true, inputName=HueShift, inputType=FunctionInput_Scalar, sortPriority=3}
  FunctionOutput(None=SetMaterialAttributes!(
    MaterialAttributes=@n1.MaterialAttributes
    EmissiveColor=MaterialFunctionCall(Hue Shift Percentage=@n2, Texture=@n1.EmissiveColor) {materialFunction=&HueShift}.Result
   ) {attributeSetTypes [#1]}) {outputName=MA, sortPriority=1}
  FunctionOutput(None=MaterialFunctionCall(Hue Shift Percentage=@n2, Texture=FunctionInput() {inputName=InputEmissive, sortPriority=2}) {materialFunction=&HueShift}.Result) {outputName=Emissive, sortPriority=2}
 ]}
