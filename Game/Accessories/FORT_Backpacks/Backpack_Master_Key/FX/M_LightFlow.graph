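# Translucent unlit mesh-particle material for flowing light FX on the Master Key back bling.
# Vertex red blends two attribute layers: panned smoke with turbulence- and facing-based opacity plus noise vertex wobble, and a half-opaque noise-tinted layer in the same emissive colors.
# Particle alpha scales opacity before a color-shift function, and the dynamic Random parameter offsets wobble timing.
# A static switch can multiply opacity by a time-of-day value that is nonzero at night.
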Material /Game/Accessories/FORT_Backpacks/Backpack_Master_Key/FX/M_LightFlow
graph {
 properties {
  bUseMaterialAttributes=true
  bUsedWithMeshParticles=true
  blendMode=BLEND_Translucent
  pixelDepthOffsetMode=PDOM_Legacy
  refractionMethod=RM_IndexOfRefraction
  shadingModel=MSM_Unlit
  usedShadingModels=Unlit
 }
 nodes [
  n1 = DotProduct(A=VertexNormalWS(), B=Normalize(VectorInput=WorldPosition() {worldPositionShaderOffset=WPT_CameraRelative}.XYZ))
  Abs(None=@n1)
  MaterialFunctionCall(Offset Amount=Constant() {r=0.01}) {materialFunction=/Engine/Functions/Engine_MaterialFunctions02/WorldPositionOffset/CameraOffset}
  n2 = TextureCoordinate()
  n3 = VectorParameter() {parameterName=UV, defaultValue=<rgba 2 0.6 0.1 -0.3>}
  n4 = AppendVector(A=@n3.R, B=@n3.R)
  n5 = VertexInterpolator(VS=Panner(Coordinate=Multiply(A=@n2, B=@n4), Speed=AppendVector(A=@n3.B, B=@n3.B)))
  n6 = VectorParameter() {parameterName=Emissive2, defaultValue=<rgba 10 0.9312 0 1>}
  n7 = VectorParameter() {parameterName=Emissive, defaultValue=<rgba 50 13.2812 1.0417 1>}
  n8 = VertexColor()
  n9 = Multiply(
   A=Desaturation(
    None=TextureSample(UVs=@n5.PS) {texture=&T_Inky_Smoke_Tile_Chromatic}.RGB
    Fraction=Constant() {r=0.5}
   )
   B=LinearInterpolate(A=@n6.RGB, B=@n7.RGB, Alpha=@n8.R)
  )
  MaterialFunctionCall(Emissive=@n9) {materialFunction=&MF_FX_EmissiveAdjustments}
  n10 = TextureCoordinate()
  n11 = VectorParameter() {parameterName=UV2, defaultValue=<rgba 1 0.2 -0.05 -0.1>}
  n12 = AppendVector(A=@n11.R, B=@n11.R)
  n13 = VertexInterpolator(VS=Panner(Coordinate=Multiply(A=@n10, B=@n12), Speed=AppendVector(A=@n11.B, B=@n11.B)))
  n14 = Multiply(
   A=@n8.R
   B=LinearInterpolate(A=@n8.R, Alpha=TextureSample(UVs=@n13.PS) {texture=/Game/Global/Textures/Default/Caustics/T_Turbulence_Seamless}.R)
  )
  n15 = Multiply(A=@n14, B=@n14)
  n16 = Multiply(A=Multiply(A=Multiply(A=@n15, B=@n15), B=Multiply(A=@n1, B=@n1))) {constB=0.25}
  n17 = Panner(
   Coordinate=TextureCoordinate() {coordinateIndex=1, uTiling=64}
   Time=Add(A=Time(), B=DynamicParameter() {paramNames [Param1,Random,Param3,HueShift]}.Random)
  ) {speedX=0.1, speedY=0.1}
  n18 = Multiply(A=Subtract(A=TextureSample(UVs=@n17) {samplerType=SAMPLERTYPE_LinearColor, texture=&T_NoisePack_LinearColor}.RGB) {constB=0.5}) {constB=5}
  n19 = Panner(Coordinate=TextureCoordinate() {coordinateIndex=1, uTiling=2, vTiling=2}) {speedX=0.5, speedY=0.5}
  n20 = Multiply(A=LinearInterpolate(
    A=@n6.RGB
    B=@n7.RGB
    Alpha=TextureSample(UVs=@n19) {samplerType=SAMPLERTYPE_LinearColor, texture=&T_NoisePack_LinearColor}.R
   )) {constB=0.5}
  n21 = Constant() {r=0.5}
  n22 = VertexColor()
  n23 = Saturate(None=Multiply(A=Subtract(A=@n22.R) {constB=0.5}) {constB=2})
  n24 = ScalarParameter() {defaultValue=10, parameterName=Exp}
  n25 = GetMaterialAttributes(None=BlendMaterialAttributes(
    A=SetMaterialAttributes!(MaterialAttributes, Emissive Color=@n9, Opacity=@n16, World Position Offset=@n18) {attributeSetTypes [#1, #2, #3]}
    B=SetMaterialAttributes(Emissive Color=@n20, Opacity=@n21) {attributeSetTypes [#1, #2, #3]}
    Alpha=Power(Base=@n23, Exp=@n24)
   )) {attributeGetTypes [#2]}
  n26 = ParticleColor()
  n27 = MaterialFunctionCall!(
   MA=SetMaterialAttributes!(MaterialAttributes=@n25.MaterialAttributes, Opacity=Multiply(A=@n25.Opacity, B=@n26.A)) {attributeSetTypes [#2]}
   InputEmissive
   HueShift
  ) {materialFunction=&/MF_ColorShift}
  n28 = BreakMaterialAttributes!(Attr=@n27.MA)
  n29 = Constant()
  n30 = MaterialFunctionCall(Color_Morning=@n29, Color_Day=@n29, Color_Evening=@n29, Color_Night=Constant() {r=1}) {materialFunction=&Color_Time_of_Day}
  StaticSwitchParameter(
   True=MakeMaterialAttributes!(
    BaseColor=@n28.BaseColor
    Metallic=@n28.BaseColor
    Specular=@n28.BaseColor
    Roughness=@n28.BaseColor
    Anisotropy
    EmissiveColor=@n28.BaseColor
    Opacity=Multiply(A=@n28.Opacity, B=@n30.Result)
    OpacityMask=@n28.BaseColor
    Normal=@n28.BaseColor
    Tangent
    WorldPositionOffset=@n28.BaseColor
    SubsurfaceColor=@n28.BaseColor
    ClearCoat=@n28.BaseColor
    ClearCoatRoughness=@n28.BaseColor
    AmbientOcclusion=@n28.BaseColor
    Refraction=@n28.BaseColor
    CustomizedUVs_0
    CustomizedUVs_1
    CustomizedUVs_2
    CustomizedUVs_3
    CustomizedUVs_4
    CustomizedUVs_5
    CustomizedUVs_6
    CustomizedUVs_7
    PixelDepthOffset=@n28.BaseColor
    ShadingModel
    Displacement
   )
   False=@n27.MA
  ) {parameterName=Alpha time of day}
 ]
}
