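# Translucent, unlit, two-sided particle material for light rays on the Master Key backpack, usable on mesh, sprite and ribbon effects.
# Emissive multiplies a tint, particle color and a UV- and time-driven hue shift, then runs through color-shift and emissive-adjustment functions.
# Opacity combines noise panned by Dynamic Parameter time, a steep V falloff, a swizzled ramp, particle alpha and view-facing edge fade.
# Depth fade is skipped on mobile.
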Material /Game/Accessories/FORT_Backpacks/Backpack_Master_Key/FX/M_Rays
graph {
 properties {
  bUsedWithMeshParticles=true
  bUsedWithNiagaraMeshParticles=true
  bUsedWithNiagaraRibbons=true
  bUsedWithNiagaraSprites=true
  blendMode=BLEND_Translucent
  pixelDepthOffsetMode=PDOM_Legacy
  shadingModel=MSM_Unlit
  twoSided=true
  usedShadingModels=Unlit
 }
 nodes [
  n1 = DynamicParameter() {defaultValue=<rgba 1 1 2 0>, paramNames [Time,Random,AlphaExp,HueShift]}
  n2 = Add(A=TextureCoordinate(), B=Add(A=@n1.Time, B=@n1.Time))
  n3 = Multiply(
   A=VectorParameter() {parameterName=Emissive, defaultValue=<rgba 12 3.181 1.0171 1>}.RGB
   B=MaterialFunctionCall(Hue Shift Percentage=ComponentMask(None=@n2) {g=true}, Texture=Constant3Vector() {constant=<rgba 1 0.5 0.5 0>}) {materialFunction=&HueShift}.Result
  )
  n4 = ParticleColor()
  MaterialFunctionCall(Emissive=MaterialFunctionCall!(MA, InputEmissive=Multiply(A=@n3, B=@n4.RGB), HueShift) {materialFunction=&/MF_ColorShift}.Emissive) {materialFunction=&MF_FX_EmissiveAdjustments}
  n5 = VectorParameter() {parameterName=UVandSpeed, defaultValue=<rgba 0.5 0.25 0 -1>}
  n6 = DynamicParameter() {defaultValue=<rgba 1 1 2 0>, paramNames [Time,Random,AlphaExp,HueShift]}
  n7 = Panner(
   Coordinate=Multiply(A=TextureCoordinate(), B=AppendVector(A=@n5.R, B=@n5.R))
   Time=Add(A=@n6.Time, B=@n6.Time)
   Speed=AppendVector(A=@n5.B, B=@n5.B)
  )
  n8 = TextureCoordinate()
  n9 = ComponentMask(None=OneMinus(None=@n8)) {g=true}
  n10 = Multiply(A=@n9, B=@n9)
  n11 = Multiply(A=@n10, B=@n10)
  n12 = TextureSample(UVs=MaterialFunctionCall(XYZ=@n8, XY=@n8) {materialFunction=&Swizzle}.YX) {texture=/Game/Effects/Fort_Effects/Textures/Ramps/T_Ramp_Smooth}
  n13 = VertexNormalWS()
  n14 = CameraVectorWS()
  n15 = Multiply(A=ParticleColor().A, B=Abs(None=DotProduct(A=@n13, B=@n14)))
  n16 = DepthFade(FadeDistance=ScalarParameter() {parameterName=DF_FD})
  Multiply(
   A=Multiply(
    A=TextureSampleParameter2D(UVs=@n7) {parameterName=Noise, samplerType=SAMPLERTYPE_LinearColor, texture=&T_NoisePack_LinearColor}.R
    B=Multiply(A=@n11, B=@n12.R)
   )
   B=ShadingPathSwitch(Default=Multiply(A=@n15, B=@n16), Mobile=@n15) {desc=Migrated from FeatureLevelSwitch}
  )
 ]
}
